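# Simple Default Lit, Nanite-enabled material for puzzle meshes in the Canary TempleRuins dev test map.
# A Color vector parameter, white by default, feeds two multiplies: one by a Glow scalar for an adjustable glow strength, the other by a small constant for a darkened tint.
# It samples no textures, so instances vary only through Color and Glow.
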Material /Game/Athena/Artemis/Maps/Test/Dev/ChrisK/TempleRuins/Meshes/Canary_PuzzleMaterialTemple
graph {
 properties {bUsedWithNanite=true, pixelDepthOffsetMode=PDOM_Legacy, usedShadingModels=Default Lit}
 nodes [
  Constant() {r=1}
  n1 = VectorParameter() {parameterName=Color, defaultValue=<rgba 1 1 1 0>}
  Multiply(A=@n1.RGB, B=ScalarParameter() {parameterName=Glow})
  Multiply(A=@n1.RGB, B=Constant() {r=0.125})
 ]
}
